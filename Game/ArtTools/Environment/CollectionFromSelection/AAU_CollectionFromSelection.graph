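# Editor utility that creates a local collection with a given name from the assets behind the actors selected in the level.
# CreateCollection gathers static mesh asset paths from StaticMeshActors, and for other actors strips the _C suffix from the class path to get the Blueprint asset.
# If the collection exists, a Yes/No warning gates emptying it; otherwise it is created.
# The gathered soft object paths are then added to the collection.
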EditorUtilityBlueprint /Game/ArtTools/Environment/CollectionFromSelection/AAU_CollectionFromSelection
graphs [
 CreateCollection {nodes [
   n1 = CreateCollection!() -> (then, Collection Name:String) {is=title K2Node_FunctionEntry}
   > n2 = Sequence!(execute) -> (then_0, then_1, then_2:Exec, then_3:Exec)
   > n3 = GetSelectedActors()
   > n4 = Set L_AR_SelectedActors(L_AR_SelectedActors:Array of Actor Object References=@n3) -> (Output_Get:Array of Actor Object References)
   > n5 = For Each Loop(Array:Array of Actor Object References=@n4) -> (Array Element:Actor Object Reference)
   > n6 = Cast To StaticMeshActor!(execute, Object=@n5) -> (then, CastFailed, AsStatic Mesh Actor:Static Mesh Actor Object Reference) {is=title K2Node_DynamicCast}
   n7 = Get StaticMeshComponent(self=@n6)
   n8 = Break Asset Data(AssetData=CreateAssetData(InAsset=Get StaticMesh(self=@n7), bAllowBlueprintClass=false))
   n9 = "To String (Name)"(InName=@n8)
   n10 = "To String (Name)"(InName=@n8.AssetName)
   > Add(
    TargetArray:Array of Soft Object Path Structures=Get L_AR_SoftPaths() -> (L_AR_SoftPaths:Array of Soft Object Path Structures)
    NewItem:"Soft Object Path Structure (by ref)"=MakeSoftObjectPath!(PathString:String=Append!(A=@n9, B=., C:String=@n10) -> (ReturnValue)) -> (ReturnValue:Soft Object Path Structure)
   ) {is=title K2Node_CallArrayFunction}
   n11 = BreakSoftClassPath!(InSoftClassPath:Soft Class Path Structure=GetSoftClassPath!(Class:Object Class Reference=Get Class(Object=@n5)) -> (ReturnValue:Soft Class Path Structure)) -> (PathString:String)
   Add(
    execute=@n6.CastFailed
    TargetArray=Get L_AR_SoftPaths()
    NewItem=MakeSoftObjectPath(PathString=Split(SourceString=@n11, InStr=_C, SearchDir=FromEnd))
   )
   Set CollectionName(execute=@n2.then_1, CollectionName:Name=String To Name(InString=@n1)) -> (Output_Get:Name=None)
   n12 = Get CollectionManagerScriptingSubsystem()
   n13 = Get CollectionName() -> (CollectionName:Name=None)
   n14 = CollectionExists!(
    execute=@n2.then_2
    self:Collection Manager Subsystem Object Reference=@n12
    Container:Collection Container Source Structure
    Collection:Name=@n13
    ShareType:ECollectionScriptingShareType Enum=Local
   ) -> (then, ReturnValue:bool)
   > n15 = Branch(Condition=@n14)
   > n16 = Show Message Dialog(Title=Get Title() -> (Title:Text), Message=Get Message() -> (Message:Text), MessageType=YesNo, DefaultValue=No, MessageCategory=Warning)
   > n17 = Switch on EAppReturnType(Selection=@n16)
   > Return Node()
   EmptyCollection!(
    execute=@n17.Yes
    self:Collection Manager Subsystem Object Reference=Get CollectionManagerScriptingSubsystem()
    Collection:"Collection Structure (by ref)"=Make Collection!(Container:Name=Game, Name:Name=Get CollectionName() -> (CollectionName=None), ShareType:ECollectionScriptingShareType Enum=Local) -> (CollectionScriptingRef:Collection Structure) {is=title K2Node_MakeStruct}
   ) -> (then, ReturnValue:bool)
   CreateCollection!(
    execute=@n15.else
    self:Collection Manager Subsystem Object Reference=@n12
    Container:Collection Container Source Structure
    Collection:Name=@n13
    ShareType:ECollectionScriptingShareType Enum=Local
   ) -> (then, OutNewCollection:Collection Structure, ReturnValue:bool) {is=title K2Node_CallFunction}
   AddAssetsToCollection!(
    execute=@n2.then_3
    self:Collection Manager Subsystem Object Reference=@n12
    Collection:"Collection Structure (by ref)"=Make Collection(Container=Game, Name=@n13, ShareType=Local)
    AssetPaths:Array of Soft Object Path Structures=Get L_AR_SoftPaths()
   ) -> (then, ReturnValue:bool)
   Break Asset Data(AssetData=For Each Loop(Array:Array of Asset Data Structures=GetSelectedAssetData()) -> (Array Element:Asset Data Structure))
   GetSelectedAssets()
   CreateAssetData(bAllowBlueprintClass=false)
   Show Object Dialog!(
    execute
    Title:"Text (by ref)"
    InOutObject:Object Reference
    Options:"Editor Dialog Library Object Details View Options Structure (by ref)"=()
   ) -> (then, ReturnValue:bool)
   Split(SourceString=BreakSoftClassPath(InSoftClassPath=GetSoftClassPath()), InStr=., SearchDir=FromEnd)
  ]}
 EventGraph {links [], nodes []}
]
variables [
 CollectionName {type=string, default=None}
 Message {type=string, default="Collection with same name exists. \r\n\r\nContinue?"}
 Title {type=string, default=Warning}
]
